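# TextureVariation material function that breaks up visible tiling by outputting randomly shifted UVs for ordinary texture samples.
# Scaled noise splits the surface into bands, each fetching its own and its neighbor's random offset from one lookup texture, with optional rotation and scale.
# Heightmap differences and optional temporal dither shift the hard choice between both UV sets.
# Raw-UV derivatives avoid mip seams; vertex stage samples explicit mips.
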MaterialFunction /Engine/Functions/Engine_MaterialFunctions03/Texturing/TextureVariation
graph {
 properties {
  bExposeToLibrary=true
  description="Gives Texture UV variation that can be used to break up tiling of repetitive textures. It is designed to use displacement maps to create an intersection based on height, and then does a temporal dither. This avoids needing to pass in every texture to be randomized as a texture object and instead the \"Shifted UVs\" output can be used for regular texture samples. Make sure to set texture samples to MipValueMode: Derivative. Additional Parameters such as Dithering and Scale Variance can be adjusted by creating a material instance of this material."
  libraryCategoriesText [Texturing]
 }
 nodes [
  n1 = FunctionInput() {
   bUsePreviewValueAsDefault=true
   description=How much the displacement affects the blend edge location.
   inputName=Heightmap Influence
   inputType=FunctionInput_Scalar
   sortPriority=4
   previewValue=<0.5 0 0 1>
  }
  n2 = FunctionInput(Preview=MaterialFunctionCall!(Scale=Constant() {r=1000}) {materialFunction=/Engine/Functions/Engine_MaterialFunctions01/Coordinates/WorldPosition-XY}.Result) {
   bUsePreviewValueAsDefault=true
   description=The base input UVs
   inputName=UVs
   inputType=FunctionInput_Vector2
   sortPriority=1
   previewValue=<1000 0 0 1>
  }
  n3 = FunctionOutput(None=@n2) {
   description="To fix mip map artifacts, set your texture samples to MipValueMode: Derivative and use DDX, DDY from this Raw UVs pin. This will ensure the mip maps are calculated from clean unmodified UVs."
   outputName=Raw UVs
   sortPriority=2
  }
  n4 = DDX(Value=@n3)
  n5 = DDY(Value=@n3)
  n6 = FunctionInput() {
   bUsePreviewValueAsDefault=true
   description=This is how large the variation pattern should be relative to the base texture.
   inputName=Variation Scale
   inputType=FunctionInput_Scalar
   sortPriority=2
   previewValue=<3.236 0 0 1>
  }
  n7 = TextureSampleParameter2D(UVs=Divide(A=@n2, B=@n6)) {
   group=Texture Variation
   parameterName=RandomVariationPattern
   samplerSource=SSM_Wrap_WorldGroupSettings
   samplerType=SAMPLERTYPE_LinearGrayscale
   texture=/Engine/Functions/Engine_MaterialFunctions02/ExampleContent/Textures/LowResBlurredNoise
  }
  n8 = FunctionInput() {
   bUsePreviewValueAsDefault=true
   description=This inputs allows you to change how many random variation bands there are.
   inputName=Variation Levels
   inputType=FunctionInput_Scalar
   sortPriority=3
   previewValue=<6 0 0 1>
  }
  n9 = Multiply(A=@n7.R, B=@n8)
  n10 = Divide(A=Add(A=Floor(None=@n9) {desc="Stripe ##"}, B=ScalarParameter() {defaultValue=9, parameterName=Random Stripe Offset})) {constB=32}
  n11 = Divide(A=Constant() {r=0.5}) {constB=32}
  n12 = TextureSample(UVs=AppendVector(A=Add(A=@n10, B=@n11) {constB=0.5}, B=Constant() {r=0.5})) {samplerType=SAMPLERTYPE_LinearColor, texture=/Engine/EngineMaterials/RandomVectors_WithXNeighbor}
  n13 = ComponentMask(None=ComponentMask(None=@n12.RGBA) {a=true, b=true, desc=Right}) {g=true, r=true}
  n14 = Add(A=@n3, B=Multiply(A=@n13, B=ScalarParameter() {defaultValue=8, group=Texture Variation, parameterName=Random Offset Variation}))
  n15 = ComponentMask(None=@n13) {r=true}
  n16 = Add(A=Multiply(A=Subtract(A=@n15) {constB=0.5}, B=ScalarParameter() {defaultValue=-0.3, group=Texture Variation, parameterName=Random Scale Variation}))
  n17 = FunctionInput(Preview=StaticBool()) {
   bUsePreviewValueAsDefault=true
   description="Whether to randomly rotate the texture samples. When enabled, this can cause issues with anisotropic Mip Maps. If this causes issues, consider using the function \"Compute Mip Level\" as an alternative to DDX,DDY."
   inputName=Random Rotation and Scale
   inputType=FunctionInput_StaticBool
   sortPriority=6
  }
  n18 = StaticSwitch(
   True=MaterialFunctionCall!(UVs=Multiply(A=@n14, B=@n16), Offset, Vector=@n13) {materialFunction=&/TextureVariation_RotateUV}.Result
   False=@n14
   Value=@n17
  )
  n19 = FunctionInput(Preview=TextureObject() {samplerType=SAMPLERTYPE_LinearColor, texture=/Engine/ArtTools/RenderToTexture/Textures/127grey}) {
   bUsePreviewValueAsDefault=true
   description=This should contain a Texture Object with a Displacement map. Make sure to specify the Height Channel to match your texture.
   inputName=Heightmap
   inputType=FunctionInput_Texture2D
  }
  n20 = ComponentMask(None=ComponentMask(None=@n12.RGB) {desc=Left, g=true, r=true}) {g=true, r=true}
  n21 = Add(A=@n3, B=Multiply(A=@n20, B=ScalarParameter() {defaultValue=8, group=Texture Variation, parameterName=Random Offset Variation})) {desc=Raw UV + Offset}
  n22 = ComponentMask(None=@n20) {r=true}
  n23 = Add(A=Multiply(A=Subtract(A=@n22) {constB=0.5}, B=ScalarParameter() {defaultValue=-0.5, group=Texture Variation, parameterName=Random Scale Variation}))
  n24 = StaticSwitch(
   True=MaterialFunctionCall(UVs=Multiply(A=@n21, B=@n23), Vector=@n20) {materialFunction=&TextureVariation_RotateUV}.Result
   False=@n21
   Value=@n17
  )
  n25 = TextureSample!(
   UVs=@n24
   Tex=@n19
   "DDX(UVs)"=FunctionOutput(None=DDX(Value=@n3)) {outputName=DDX, sortPriority=4}
   "DDY(UVs)"=FunctionOutput(None=DDY(Value=@n3)) {outputName=DDY, sortPriority=5}
   Apply View MipBias
  ) {constMipValue=0, mipValueMode=TMVM_Derivative, samplerSource=SSM_Wrap_WorldGroupSettings}
  n26 = FunctionInput(Preview=StaticBool() {value=true}) {
   bUsePreviewValueAsDefault=true
   description="Reads a second displacement sample so that at cell boundaries, both heights will be used when making a height comparison for a more accurate intersection. When false, only one displacement value is used."
   inputName=HQ Edge Comparison
   inputType=FunctionInput_StaticBool
   sortPriority=9
   previewValue=<0.5 0 0 1>
  }
  n27 = StaticSwitch(
   True=Subtract(
    A=TextureSample(UVs=@n18, Tex=@n19, "DDX(UVs)"=@n4, "DDY(UVs)"=@n5) {constMipValue=0, mipValueMode=TMVM_Derivative, samplerSource=SSM_Wrap_WorldGroupSettings}.RGBA
    B=@n25.RGBA
   )
   False=Subtract(A=Constant() {r=0.5}, B=@n25.RGBA)
   Value=@n26
  )
  n28 = ScalarParameter() {group=Texture Variation, parameterName=Texture Variation Height VS Mip}
  n29 = Subtract(
   A=TextureSample!(UVs=@n18, Tex=@n19, Level=@n28, Apply View MipBias) {constMipValue=0, mipValueMode=TMVM_MipLevel, samplerSource=SSM_Wrap_WorldGroupSettings}.R
   B=TextureSample(UVs=@n24, Tex=@n19, Level=@n28) {constMipValue=0, mipValueMode=TMVM_MipLevel, samplerSource=SSM_Wrap_WorldGroupSettings}.R
  )
  n30 = FunctionInput() {
   bUsePreviewValueAsDefault=true
   description=Which channel to use
   inputName=Mask Channel
   inputType=FunctionInput_Vector4
   sortPriority=5
   previewValue=<1 0 0 0>
  }
  n31 = Multiply(A=@n1, B=DotProduct(A=ShaderStageSwitch(PixelShader=@n27, VertexShader=@n29), B=@n30))
  n32 = Multiply(
   A=MaterialFunctionCall(Alpha Threshold=Constant()) {materialFunction=&DitherTemporalAA}.Result
   B=ScalarParameter() {defaultValue=0.2, group=Texture Variation, parameterName=Texture Variation Dithering}
  )
  n33 = FunctionInput(Preview=StaticBool()) {
   bUsePreviewValueAsDefault=true
   description=Enabling Dither uses a pattern to alternate between layers in a way that Temporal Anti Aliasing can smooth out.
   inputName=Use Dither
   inputType=FunctionInput_StaticBool
   sortPriority=6
  }
  n34 = Add(A=Frac(None=@n9), B=StaticSwitch(True=Multiply(A=@n32) {constB=0.5}, False=Constant(), Value=@n33))
  n35 = Saturate(None=Round(None=Add(A=@n34, B=@n31)))
  FunctionOutput(None=ComponentMask(None=LinearInterpolate(A=@n24, B=@n18, Alpha=@n35)) {g=true, r=true}) {description="The randomly shifted UVs. Plug this pin into your texture UVs. ", outputName=Shifted UVs}
  FunctionOutput(None=LinearInterpolate(A=@n20, B=@n13, Alpha=@n35)) {
   description=This is the random UV offset given to each random sample. The B channel contains the raondom Rotation used to fix up Normals.
   outputName=Random Offset
   sortPriority=5
  }
 ]
}
