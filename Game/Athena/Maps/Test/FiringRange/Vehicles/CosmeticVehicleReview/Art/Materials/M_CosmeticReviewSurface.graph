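# Default Lit surface material for the cosmetic vehicle review area of the Firing Range test map.
# World-aligned projection samples color, normal and packed MSR textures, each with a TextureSize parameter.
# Channel-mask parameters split MSR into specular, metallic and roughness, while static switches allow constant metallic or base color instead of textures.
# TextureVariation, UVLayoutToWorldSpacePosition and a Helios forest-floor sample have no visible connections.
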Material /Game/Athena/Maps/Test/FiringRange/Vehicles/CosmeticVehicleReview/Art/Materials/M_CosmeticReviewSurface
graph {
 properties {bAutomaticallySetUsageInEditor=false, pixelDepthOffsetMode=PDOM_Legacy, usedShadingModels=Default Lit}
 nodes [
  n1 = MaterialFunctionCall(
   TextureObject=TextureObjectParameter() {group=MSR, parameterName=MSR, samplerType=SAMPLERTYPE_Masks, texture=&T_BlackMask}
   TextureSize=ScalarParameter() {defaultValue=256, parameterName=TextureSize}
  ) {materialFunction=&WorldAlignedTexture}
  ChannelMaskParameter(None=@n1.XYZ Texture) {group=MSR, parameterName=SpecularChannel}
  ChannelMaskParameter(None=@n1.XYZ Texture) {group=MSR, maskChannel=Blue, parameterName=RoughChannel, defaultValue=<rgba 0 0 1 0>}
  n2 = TextureObjectParameter() {group=Normal, parameterName=Normal, samplerType=SAMPLERTYPE_Normal, texture=&FlatNormalBlue}
  MaterialFunctionCall(TextureObject=@n2, TextureSize=ScalarParameter() {defaultValue=256, parameterName=TextureSize}) {materialFunction=&WorldAlignedNormal}
  MaterialFunctionCall!(
   Heightmap
   UVs
   Variation Scale
   Variation Levels
   Heightmap Influence
   Mask Channel
   Use Dither
   Random Rotation and Scale
   HQ Edge Comparison
  ) {materialFunction=/Engine/Functions/Engine_MaterialFunctions03/Texturing/TextureVariation}
  MaterialFunctionCall!(Texture Coordinate) {materialFunction=/Engine/Functions/Engine_MaterialFunctions02/WorldPositionOffset/UVLayoutToWorldSpacePosition}
  StaticSwitchParameter(
   True=ChannelMaskParameter(None=@n1.XYZ Texture) {group=MSR, maskChannel=Green, parameterName=MetalChannel, defaultValue=<rgba 0 1 0 0>}.R
   False=ScalarParameter() {parameterName=MetallicConstant}
  ) {defaultValue=true, group=MSR, parameterName=UseMetalTexture}
  StaticSwitch(
   True=MaterialFunctionCall(
    TextureObject=TextureObjectParameter() {group=Color, parameterName=Color, texture=/Game/Global/Textures/Default/Blanks/T_Blank_LegalWhite_sRGBColor}
    TextureSize=ScalarParameter() {defaultValue=256, parameterName=TextureSize}
   ) {materialFunction=&WorldAlignedTexture}.XYZ Texture
   False=VectorParameter() {parameterName=BaseColor, defaultValue=<rgba 0.1217 0.1217 0.1217 1>}.RGB
   Value=StaticBoolParameter() {defaultValue=true, parameterName=UseBaseColorTexture}
  )
  TextureSample() {texture=/Game/Environments/Helios/Terrain/Textures/T_Helios_Bor_ForestFloor_02_D}
 ]
}
